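# Sound cue for a wooden village door closing, built from two layered door-close sounds.
# A mixer combines a randomly picked heavy wood close with a delayed, randomly picked residential close, each layer passing through its own modulator.
# All quality levels share this mix, and both random nodes are excluded from branch culling.
# It uses the Fort doors attenuation and sound class, with a concurrency override that prevents new instances at the limit.
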SoundCue /Game/Sounds/Asteria/Environment_Actors/Door_Wood/Door_Wood_Village_Close_Cue
root=QualityLevel_0
nodes [
 {id=Delay_0, children [Random_1], properties {delayMin=0.2}}
 {id=Mixer_0, children [Modulator_0,Modulator_2], properties {inputVolume [1,1]}}
 {id=Modulator_0, children [Delay_0]}
 {id=Modulator_2, children [Random_0]}
 {id=QualityLevel_0, children [Mixer_0,Mixer_0,Mixer_0,Mixer_0]}
 {id=Random_0, children [WavePlayer_0,WavePlayer_1], properties {bShouldExcludeFromBranchCulling=true, weights [1,1]}}
 {id=Random_1, children [WavePlayer_2,WavePlayer_3], properties {bShouldExcludeFromBranchCulling=true, weights [1,1]}}
 {id=WavePlayer_0, properties {soundWaveAssetPtr=/Game/Sounds/Fort_Doors/World/Door_Wood_Heavy_Close_01}}
 {id=WavePlayer_1, properties {soundWaveAssetPtr=/Game/Sounds/Fort_Doors/World/Door_Wood_Heavy_Close_02}}
 {id=WavePlayer_2, properties {soundWaveAssetPtr=/Game/Sounds/Fort_Doors/World/door_residential_close_F1}}
 {id=WavePlayer_3, properties {soundWaveAssetPtr=/Game/Sounds/Fort_Doors/World/door_residential_close_F2}}
]
properties {
 attenuationSettings=/Game/Sounds/Attenuation/Fort_Doors_Att
 bOverrideConcurrency=true
 maxDistance=2400
 soundClassObject=&Fort_Doors
 volumeMultiplier=0.5
 attenuationOverrides {falloffDistance=1600}
 concurrencyOverrides {resolutionRule=PreventNew}
}
