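# Sound cue for the opening of a village wooden door, judging by its name, layered from residential and heavy wood door recordings.
# A mixer blends one of four residential opens, at reduced volume, with one of two heavy wood opens behind a delay node.
# A root modulator processes the mix for variation.
# It uses the Fort_Doors sound class and attenuation, with a concurrency override that prevents new instances at the limit.
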SoundCue /Game/Sounds/Asteria/Environment_Actors/Door_Wood/Door_Wood_Village_Open_Cue
root=Modulator_1
nodes [
 {id=Delay_0, children [Random_0], properties {delayMin=0.2}}
 {id=Mixer_0, children [Random_1,Delay_0], properties {inputVolume [0.7,1]}}
 {id=Modulator_1, children [Mixer_0]}
 {id=Random_0, children [WavePlayer_0,WavePlayer_1], properties {weights [1,1]}}
 {id=Random_1, children [WavePlayer_2,WavePlayer_3,WavePlayer_4,WavePlayer_5], properties {weights [1,1,1,1]}}
 {id=WavePlayer_0, properties {soundWaveAssetPtr=/Game/Sounds/Fort_Doors/World/Door_Wood_Heavy_Open_01}}
 {id=WavePlayer_1, properties {soundWaveAssetPtr=/Game/Sounds/Fort_Doors/World/Door_Wood_Heavy_Open_02}}
 {id=WavePlayer_2, properties {soundWaveAssetPtr=/Game/Sounds/Fort_Doors/World/door_residential_open_G1}}
 {id=WavePlayer_3, properties {soundWaveAssetPtr=/Game/Sounds/Fort_Doors/World/door_residential_open_G2}}
 {id=WavePlayer_4, properties {soundWaveAssetPtr=/Game/Sounds/Fort_Doors/World/door_residential_open_G3}}
 {id=WavePlayer_5, properties {soundWaveAssetPtr=/Game/Sounds/Fort_Doors/World/door_residential_open_G4}}
]
properties {
 attenuationSettings=/Game/Sounds/Attenuation/Fort_Doors_Att
 bOverrideConcurrency=true
 maxDistance=2400
 soundClassObject=&Fort_Doors
 volumeMultiplier=0.5
 attenuationOverrides {falloffDistance=1600}
 concurrencyOverrides {resolutionRule=PreventNew}
}
